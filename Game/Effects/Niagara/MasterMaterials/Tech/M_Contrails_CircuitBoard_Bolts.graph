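# Translucent unlit Niagara master material for ribbons and sprites, drawing scrolling circuit-board bolts, apparently for contrails.
# Sprite Picker blends four channel combinations of a branching circuit texture; the second subtracts green from itself, yielding zero.
# A panning gradient mask, offset and scrolled by dynamic parameters and shaped by power and strength, multiplies that pattern.
# Particle color and an inverted depth SphereMask that fades pixels near the camera shape emissive color and opacity.
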Material /Game/Effects/Niagara/MasterMaterials/Tech/M_Contrails_CircuitBoard_Bolts
graph {
 properties {
  bAutomaticallySetUsageInEditor=false
  bUsedWithNiagaraRibbons=true
  bUsedWithNiagaraSprites=true
  blendMode=BLEND_Translucent
  shadingModel=MSM_Unlit
  usedShadingModels=Unlit
 }
 nodes [
  n1 = OneMinus(None=SphereMask(A=PixelDepth().R, B=Constant()) {attenuationRadius=32, hardnessPercent=10})
  n2 = ParticleColor()
  n3 = TextureSampleParameter2D(UVs=TextureCoordinate()) {
   group=Sprite
   parameterName=Sprite
   samplerType=SAMPLERTYPE_LinearColor
   texture=/Game/Effects/Niagara/MasterMaterials/Tech/Textures/T_Circuits_Branches_Diffuse
  }
  n4 = Subtract(A=@n3.G, B=@n3.G)
  n5 = Add(A=@n3.B, B=@n3.B)
  n6 = TextureCoordinate()
  n7 = AppendVector(A=Constant() {r=1}, B=ScalarParameter() {defaultValue=1, group=Mask, parameterName=Mask Tile V, sortPriority=34})
  n8 = DynamicParameter() {paramNames [V Start Offset, V Scroll, Sprite Picker, Mask Scroll Speed]}
  n9 = Add(A=ComponentMask(None=Multiply(A=@n6, B=@n7)) {g=true}, B=AppendVector(A=Constant(), B=@n8.V Start Offset))
  n10 = ScalarParameter() {parameterName=Mask Scroll U}
  n11 = Add(A=@n8.Mask Scroll Speed, B=ScalarParameter() {defaultValue=1, parameterName=Mask Scroll V})
  n12 = TextureSampleParameter2D(UVs=Panner(Coordinate=Add(A=@n9, B=@n8.V Scroll), Speed=AppendVector(A=@n10, B=@n11))) {
   group=Mask
   parameterName=Mask
   samplerType=SAMPLERTYPE_Masks
   sortPriority=35
   texture=/Game/Global/Textures/Default/Masks_and_Gradients/T_VertGradient_MASK_2
  }
  n13 = ScalarParameter() {defaultValue=0.5, group=Mask, parameterName=Mask Pow}
  n14 = Multiply(
   A=MaterialFunctionCall!(
    Mask 1=Add(A=@n3.R, B=@n3.R) {constB=0}
    Mask 2=Clamp(None=@n4)
    Mask 3=@n5
    Mask 4=Add(A=@n3.G, B=@n5)
    Blend=DynamicParameter() {paramNames [V Start Offset, V Scroll, Sprite Picker, Mask Scroll Speed]}.Sprite Picker
   ) {materialFunction=/Game/Effects/Niagara/MasterMaterials/Functions/Utility/MF_Lerp_4_Scalar}.Result
   B=Multiply(
    A=Power(Base=@n12.R, Exp=@n13) {constExponent=1}
    B=ScalarParameter() {defaultValue=2, group=Mask, parameterName=Mask Strength, sortPriority=33}
   )
  )
  Multiply(A=@n1, B=Multiply(A=@n2.A, B=@n14))
  Multiply(
   A=Multiply(A=@n1, B=Multiply(A=@n2.RGB, B=@n14))
   B=ScalarParameter() {defaultValue=1, group=Sprite, parameterName=Emissive Strength, sortPriority=31}
  )
 ]
}
