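# Additive unlit Niagara sprite material for glitchy circuit-board speed lines on contrails, going by its name.
# The packed circuit texture's red and green, sampled in screen space and scaled by distortion strength, plus a world-height vertical offset become UVs for a second lookup of that texture.
# That lookup's blue channel, masked by a diamond gradient and tinted by particle color, forms the glow; blue times particle alpha gives opacity.
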Material /Game/Effects/Niagara/MasterMaterials/Tech/M_Contrails_CircuitBoard_GlitchSpeedLine
graph {
 properties {
  bAutomaticallySetUsageInEditor=false
  bUsedWithNiagaraSprites=true
  blendMode=BLEND_Additive
  shadingModel=MSM_Unlit
  usedShadingModels=Unlit
 }
 nodes [
  n1 = ParticleColor()
  n2 = ScreenPosition()
  n3 = Multiply(
   A=ComponentMask(None=TextureSampleParameter2D(UVs=@n2.ViewportUV) {parameterName=Packed_Texture, texture=/Game/Effects/Niagara/MasterMaterials/Tech/Textures/T_Circuits_Diffuse}.RGB) {g=true, r=true}
   B=ScalarParameter() {defaultValue=1, parameterName=Distortion_Strength}
  ) {constB=100}
  n4 = Divide(A=WorldPosition().Z) {constB=5000}
  n5 = AppendVector(A=Constant(), B=Multiply(A=OneMinus(None=@n4)) {constB=0.25})
  n6 = TextureSampleParameter2D(UVs=Multiply(A=Add(A=@n3, B=@n5)) {constB=10}) {parameterName=Packed_Texture, texture=&T_Circuits_Diffuse}
  n7 = MaterialFunctionCall!(Falloff) {materialFunction=/Engine/Functions/Engine_MaterialFunctions01/Gradient/DiamondGradient}
  n8 = ScalarParameter() {defaultValue=10, parameterName=Mask Strength}
  Multiply(A=@n1.RGB, B=Multiply(A=@n6.B, B=Multiply(A=@n7.DiamondGradient, B=@n8) {constB=10}))
  Multiply(A=@n1.A, B=@n6.B)
 ]
}
