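# Translucent unlit Niagara master material for circuit-board contrail ribbons, also usable on sprite and mesh particles.
# One packed texture supplies RG distortion and two panning masks (B, A), offset by world height, with power-warped U tiling on the first.
# Dynamic parameters drive distortion amount, offset and a glitch color shift.
# Summed masks, optionally shaped by a diamond or radial sprite mask, multiply particle color and alpha, fading near the camera.
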Material /Game/Effects/Niagara/MasterMaterials/Tech/M_Contrails_Circuitboard_Ribbon
graph {
 properties {
  bAutomaticallySetUsageInEditor=false
  bUsedWithNiagaraMeshParticles=true
  bUsedWithNiagaraRibbons=true
  bUsedWithNiagaraSprites=true
  blendMode=BLEND_Translucent
  shadingModel=MSM_Unlit
  usedShadingModels=Unlit
 }
 nodes [
  n1 = OneMinus(None=SphereMask(A=PixelDepth().R, B=Constant()) {hardnessPercent=10})
  n2 = ParticleColor()
  n3 = ScalarParameter() {group="Mask 1 (Blue)", parameterName=Mark 1 Offset U}
  n4 = Add(
   A=DynamicParameter() {defaultValue=<rgba 0 1 1 1>, paramNames [Distortion Amount, Distortion U Offset, Param3, Param4]}.Distortion U Offset
   B=@n3
  )
  n5 = Add(
   A=TextureCoordinate()
   B=AppendVector(
    A=StaticSwitchParameter(True=@n4, False=@n3) {group="Mask 1 (Blue)", parameterName=Use_NS_Detail_Offset}
    B=ScalarParameter() {group="Mask 1 (Blue)", parameterName=Mark 1 Offset V}
   )
  )
  n6 = ComponentMask(None=OneMinus(None=@n5)) {r=true}
  n7 = ScalarParameter() {defaultValue=3, group="Mask 1 (Blue)", parameterName=Mask 1 Pow Tile U}
  n8 = Multiply(
   A=@n6
   B=Multiply(A=Power(Base=@n6, Exp=@n7) {constExponent=5}, B=ScalarParameter() {defaultValue=5, group="Mask 1 (Blue)", parameterName=Mask 1 Tile U}) {constB=10}
  )
  n9 = Multiply(A=ComponentMask(None=@n5) {g=true}, B=ScalarParameter() {defaultValue=0.5, group="Mask 1 (Blue)", parameterName=Mask 1 Tile V}) {constB=0.3}
  n10 = ScalarParameter() {defaultValue=1, group="Mask 1 (Blue)", parameterName=Mask 1 Scroll Speed U}
  n11 = ScalarParameter() {group="Mask 1 (Blue)", parameterName=Mask 1 Scroll Speed V}
  n12 = WorldPosition()
  n13 = Multiply(A=OneMinus(None=Divide(A=@n12.Z) {constB=5000})) {constB=4}
  n14 = Constant()
  n15 = Add(
   A=Panner(Coordinate=AppendVector(A=@n8, B=@n9), Speed=AppendVector(A=@n10, B=@n11)) {speedX=10}
   B=NamedRerouteDeclaration(None=AppendVector(A=@n13, B=@n14)) {name=WorldOffset, nodeColor=<rgba 0.6875 0 1 1>}
  )
  n16 = TextureCoordinate()
  n17 = AppendVector(
   A=DynamicParameter() {defaultValue=<rgba 0 1 1 1>, paramNames [Distortion Amount, Distortion U Offset, Param3, Param4]}.Distortion U Offset
   B=Constant()
  )
  n18 = ScalarParameter() {defaultValue=2, group="UV Distortion (RG)", parameterName=Distort_Tile_U}
  n19 = ScalarParameter() {defaultValue=1, group="UV Distortion (RG)", parameterName=Distort_Tile_V, sortPriority=33}
  n20 = TextureSampleParameter2D(UVs=Multiply(A=Add(A=@n16, B=@n17), B=AppendVector(A=@n18, B=@n19))) {
   desc="RG - UV Distortion, BA - Masks / Detail Maps"
   group=Shared
   parameterName=Packed Texture
   texture=/Game/Effects/Niagara/MasterMaterials/Tech/Textures/T_Circuits_Diffuse
  }
  n21 = Constant() {r=-0.5}
  n22 = Constant() {r=0.5}
  n23 = Multiply(
   A=AppendVector(
    A=MaterialFunctionCall(Input=@n20.R, Target Low=@n21, Target High=@n22) {materialFunction=&RemapValueRange}.Result
    B=MaterialFunctionCall(Input=@n20.G, Target Low=@n21, Target High=@n22) {materialFunction=&RemapValueRange}.Result
   )
   B=DynamicParameter() {defaultValue=<rgba 0 1 1 1>, paramNames [Distortion Amount, Distortion U Offset, Param3, Param4]}.Distortion Amount
  )
  n24 = Add(A=@n15, B=@n23)
  n25 = TextureSampleParameter2D(UVs=StaticSwitchParameter(
    True=StaticSwitchParameter(True=@n24, False=@n15) {defaultValue=true, group=Shared, parameterName=Mask 1 UV Distortion, sortPriority=31}
    False=@n15
   ) {defaultValue=true, group="UV Distortion (RG)", parameterName=Use UV Distortion, sortPriority=31}) {
   desc="RG - UV Distortion, BA - Masks / Detail Maps"
   group=Shared
   parameterName=Packed Texture
   texture=&T_Circuits_Diffuse
  }
  n26 = ScalarParameter() {defaultValue=1, group="Mask 1 (Blue)", parameterName=Mark 1 Strength}
  n27 = TextureCoordinate()
  n28 = AppendVector(
   A=ScalarParameter() {defaultValue=3, group="Mask 2 (Alpha)", parameterName=Mask 2 Tile U}
   B=ScalarParameter() {defaultValue=0.5, group="Mask 2 (Alpha)", parameterName=Mask 2 Tile V}
  )
  n29 = ScalarParameter() {defaultValue=-1, group="Mask 2 (Alpha)", parameterName=Mask 2 Scroll U}
  n30 = ScalarParameter() {defaultValue=0.2, group="Mask 2 (Alpha)", parameterName=Mask 2 Scroll V}
  n31 = Add(A=OneMinus(None=NamedRerouteUsage()), B=Panner(Coordinate=Multiply(A=@n27, B=@n28), Speed=AppendVector(A=@n29, B=@n30)) {speedX=-1})
  n32 = Add(A=@n23, B=@n31)
  n33 = TextureSampleParameter2D(UVs=StaticSwitchParameter(True=StaticSwitchParameter(True=@n32, False=@n31) {group=Shared, parameterName=Mask 2 UV Distortion, sortPriority=30}, False=@n31) {defaultValue=true, group="UV Distortion (RG)", parameterName=Use UV Distortion, sortPriority=31}) {
   desc="RG - UV Distortion, BA - Masks / Detail Maps"
   group=Shared
   parameterName=Packed Texture
   texture=&T_Circuits_Diffuse
  }
  n34 = ScalarParameter() {defaultValue=1, group="Mask 2 (Alpha)", parameterName=Mask 2 Strength}
  n35 = StaticSwitchParameter(
   True=MaterialFunctionCall!(Falloff=ScalarParameter() {defaultValue=1, group=Sprite Mask, parameterName=Sprite Mask Falloff, sortPriority=36}) {materialFunction=/Engine/Functions/Engine_MaterialFunctions01/Gradient/DiamondGradient}.DiamondGradient
   False=MaterialFunctionCall(
    Radius=ScalarParameter() {defaultValue=0.5, group=Sprite Mask, parameterName=Sprite Mask Radius, sortPriority=35}
    Density=ScalarParameter() {defaultValue=1, group=Sprite Mask, parameterName=Sprite Mask Strength, sortPriority=34}
   ) {materialFunction=&RadialGradientExponential}.RadialGradientExponential
  ) {group=Sprite Mask, parameterName="Diamond(T) Radial(F)", sortPriority=33}
  n36 = Multiply(
   A=Add(A=Multiply(A=@n25.B, B=@n26), B=Multiply(A=@n33.A, B=@n34) {constB=0.1})
   B=StaticSwitchParameter(True=Multiply(A=@n35) {constB=5}, False=Constant() {r=1}) {defaultValue=true, group=Sprite Mask, parameterName=Use Sprite Mask}
  )
  n37 = Multiply(A=@n2.RGB, B=@n36)
  n38 = VectorParameter() {group="UV Distortion (RG)", parameterName=Glitch_Color_Shift, sortPriority=34, defaultValue=<rgba 0.5 0 0 1>}
  n39 = Multiply(A=@n36, B=@n38.RGB)
  n40 = DynamicParameter() {defaultValue=<rgba 0 1 1 1>, paramNames [Distortion Amount, Distortion U Offset, Param3, Param4]}
  Multiply(
   A=@n1
   B=StaticSwitchParameter(True=LinearInterpolate(A=@n37, B=@n39, Alpha=@n40.Distortion Amount), False=@n37) {defaultValue=true, group="UV Distortion (RG)", parameterName=Use UV Distortion, sortPriority=31}
  )
  Multiply(A=@n1, B=Multiply(A=@n2.A, B=@n36))
 ]
}
